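# Masked, two-sided character material for the Nana Eternal backpack, built between the shared 2023 CharacterParent PreFX and PostFX functions.
# The NanaMask red channel blends in a Fresnel-shaped override base color and adds Fresnel-driven emissive from squared tint colors.
# Opacity uses TwoSidedSign to clip front faces outside that mask, leaving back faces visible, while ShadowReplace keeps shadows solid.
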Material /BRCosmetics/Accessories/FORT_Backpacks/Backpack_M_MED_NanaEternal/Materials/M_NanaEternal_Backpack
graph {
 properties {
  bUseMaterialAttributes=true
  bUsedWithMeshDeformer=true
  bUsedWithMorphTargets=true
  bUsedWithNiagaraMeshParticles=true
  bUsedWithSkeletalMesh=true
  blendMode=BLEND_Masked
  pixelDepthOffsetMode=PDOM_Legacy
  shadingModel=MSM_FromMaterialExpression
  twoSided=true
  usedShadingModels=Default Lit | Subsurface | Clear Coat | Subsurface Profile | Cloth
 }
 nodes [
  Constant()
  n1 = MaterialFunctionCall() {materialFunction=&MF_CharacterParent_2023_PreFX}
  n2 = GetMaterialAttributes(None=@n1.Result) {attributeGetTypes [#1, #2]}
  n3 = TextureSampleParameter2D() {group=TwosideCull, parameterName=NanaMask, samplerType=SAMPLERTYPE_Masks, texture=&T_BlackMask}
  n4 = Saturate(None=Sign(None=TwoSidedSign()))
  n5 = Add(A=@n3.R, B=ShadowReplace(Default=LinearInterpolate(Alpha=@n4) {constA=1, constB=0}, Shadow=Constant() {r=1}))
  n6 = VectorParameter() {group=Control, parameterName=BaseColorOverride_Color, defaultValue=<rgba 0 0.5906 0.1363 1>}
  n7 = ScalarParameter() {defaultValue=2, group=TwosideCull, parameterName=Fresnel01}
  n8 = OneMinus(None=Fresnel() {baseReflectFraction=0, exponent=1})
  n9 = Multiply(A=Saturate(None=Multiply(A=@n6.RGB, B=@n6.RGB)), B=Max(A=Fresnel(ExponentIn=@n7) {baseReflectFraction=0, exponent=1}, B=Power(Base=@n8)))
  n10 = VectorParameter() {group=Banana, parameterName=FresnelColor02, defaultValue=<rgba 0.0045 1 0.7005 1>}
  n11 = Multiply(A=@n10.RGB, B=@n10.RGB)
  n12 = ScalarParameter() {defaultValue=0.1, group=Banana, parameterName=EmissiveColorOverride}
  n13 = Multiply(A=LinearInterpolate(A=Multiply(A=@n11, B=@n12), B=@n11, Alpha=Fresnel() {baseReflectFraction=0, exponent=1}) {constA=1}, B=@n3.R)
  MaterialFunctionCall(
   Material=SetMaterialAttributes!(
    MaterialAttributes=@n2.MaterialAttributes
    Opacity Mask=Saturate(None=@n5)
    Base Color=LinearInterpolate(A=@n2.Base Color, B=@n9, Alpha=@n3.R)
    Emissive Color=Add(A=@n2.Emissive Color, B=@n13)
   ) {attributeSetTypes [#3, #1, #2]}
   M=@n1.M
   Pre Skinned Local Position=@n1.M
   Distance Blend=@n1.M
   Fresnel=@n1.M
  ) {materialFunction=&MF_CharacterParent_2023_PostFX}
 ]
}
