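# Translucent, unlit overlay material for the NanaEternal backpack cosmetic, usable on skeletal meshes and Niagara mesh particles.
# Material attributes flow from the shared 2023 character PreFX parent through the NanaEternal noise function, then the TransBlur function, into the PostFX parent.
# The noise and TransBlur functions come from the matching character body's folder rather than being defined for the backpack.
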Material /BRCosmetics/Accessories/FORT_Backpacks/Backpack_M_MED_NanaEternal/Materials/M_NanaEternal_Overlay_Backpack
graph {
 properties {
  bUseMaterialAttributes=true
  bUsedWithMeshDeformer=true
  bUsedWithMorphTargets=true
  bUsedWithNiagaraMeshParticles=true
  bUsedWithSkeletalMesh=true
  blendMode=BLEND_Translucent
  pixelDepthOffsetMode=PDOM_Legacy
  shadingModel=MSM_Unlit
  usedShadingModels=Unlit
 }
 nodes [
  n1 = MaterialFunctionCall() {materialFunction=&MF_CharacterParent_2023_PreFX}
  n2 = MaterialFunctionCall!(MaterialAttributes=@n1.Result, Fresnel=@n1.Result) {materialFunction=/BRCosmetics/Characters/Player/Male/Medium/Bodies/M_MED_NanaEternal/Materials/MF_NanaEternal_Noise}
  MaterialFunctionCall(
   Material=MaterialFunctionCall!(Material=@n2.Result, Noise=@n2.Result) {materialFunction=/BRCosmetics/Characters/Player/Male/Medium/Bodies/M_MED_NanaEternal/Materials/MF_NanaEternal_TransBlur}.Result
   M=@n1.M
   Pre Skinned Local Position=@n1.M
   Distance Blend=@n1.M
   Fresnel=@n1.M
  ) {materialFunction=&MF_CharacterParent_2023_PostFX}
 ]
}
