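# Animation Blueprint for the NanaEternal backpack cosmetic, adding secondary physics motion to its dangling arms and legs.
# The input pose runs through four chained Anim Dynamics nodes (arm_L, arm_R, leg_L, leg_R) to the output pose.
# Each uses root-relative angular springs, tight angular limits and downward gravity overrides, with stiffness tuned per limb.
# A second, zero-gravity chain for the same bones is unconnected, and the event graph holds only unwired stub nodes.
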AnimBlueprint /BRCosmetics/Accessories/FORT_Backpacks/Backpack_M_MED_NanaEternal/Meshes/M_MED_NanaEternal_Pack_AnimBP
target_skeleton=&/M_MED_NanaEternal_Pack_Skeleton
graphs [
 AnimGraph {nodes [
   n1 = Anim Dynamics - Bone: arm_L!(ComponentPose, Alpha=1, GravityOverride:Vector="0,0,0") -> (Pose) {
    angularBiasOverride=1
    angularDampingOverride=0.99
    angularSpringConstant=700
    bAngularSpring=true
    bOverrideAngularDamping=true
    bUseGravityOverride=true
    simulationSpace=RootRelative
    alphaBoolBlend {blendOption=Linear}
    boundBone {boneName=arm_L}
    componentAppliedLinearAccClamp=<5000 5000 500>
    componentLinearAccScale=<0.5 0.5 0.25>
    componentLinearVelScale=<0.5 0.5 0.25>
    physicsBodyDefinitions [{
      boundBone {boneName=arm_L}
      boxExtents=<7 7 7>
      constraintSetup {angularLimitsMax=<0 25 25>, angularLimitsMin=<0 -25 -25>, angularTarget=<1 0 0>}
      localJointOffset=<0.3 0 0>
     }]
   }
   n2 = Anim Dynamics - Bone: arm_R!(ComponentPose=@n1, Alpha=1, GravityOverride:Vector="0,0,0") -> (Pose) {
    angularBiasOverride=1
    angularDampingOverride=0.99
    angularSpringConstant=1000
    bAngularSpring=true
    bOverrideAngularDamping=true
    bUseGravityOverride=true
    simulationSpace=RootRelative
    alphaBoolBlend {blendOption=Linear}
    boundBone {boneName=arm_R}
    componentAppliedLinearAccClamp=<5000 5000 500>
    componentLinearAccScale=<0.5 0.5 0.25>
    componentLinearVelScale=<0.5 0.5 0.25>
    physicsBodyDefinitions [{
      boundBone {boneName=arm_R}
      boxExtents=<7 7 7>
      constraintSetup {angularLimitsMax=<0 25 25>, angularLimitsMin=<0 -25 -25>, angularTarget=<1 0 0>}
      localJointOffset=<0.3 0 0>
     }]
   }
   n3 = Anim Dynamics - Bone: leg_L!(ComponentPose=@n2, Alpha=1, GravityOverride:Vector="0,0,0") -> (Pose) {
    angularBiasOverride=1
    angularDampingOverride=0.97
    angularSpringConstant=800
    bAngularSpring=true
    bOverrideAngularDamping=true
    bUseGravityOverride=true
    simulationSpace=RootRelative
    alphaBoolBlend {blendOption=Linear}
    boundBone {boneName=leg_L}
    componentAppliedLinearAccClamp=<5000 5000 5000>
    componentLinearAccScale=<0.5 0.5 0.25>
    componentLinearVelScale=<0.5 0.5 0.25>
    physicsBodyDefinitions [{
      boundBone {boneName=leg_L}
      boxExtents=<11 11 11>
      constraintSetup {angularLimitsMax=<0 25 25>, angularLimitsMin=<0 -25 -25>, angularTarget=<1 0 0>}
      localJointOffset=<0.3 0 0>
     }]
   }
   Anim Dynamics - Bone: leg_R!(ComponentPose=@n3, Alpha=1, GravityOverride:Vector="0,0,0") -> (Pose) {
    angularBiasOverride=1
    angularDampingOverride=0.97
    angularSpringConstant=700
    bAngularSpring=true
    bOverrideAngularDamping=true
    bUseGravityOverride=true
    simulationSpace=RootRelative
    alphaBoolBlend {blendOption=Linear}
    boundBone {boneName=leg_R}
    componentAppliedLinearAccClamp=<5000 5000 5000>
    componentLinearAccScale=<0.5 0.5 0.25>
    componentLinearVelScale=<0.5 0.5 0.25>
    physicsBodyDefinitions [{
      boundBone {boneName=leg_R}
      boxExtents=<11 11 11>
      constraintSetup {angularLimitsMax=<0 25 25>, angularLimitsMin=<0 -25 -25>, angularTarget=<1 0 0>}
      localJointOffset=<0.3 0 0>
     }]
   }
   n4 = Anim Dynamics - Bone: arm_L!(ComponentPose=Local To Component(LocalPose=Input Pose()), Alpha=1) -> (Pose) {
    angularBiasOverride=1
    angularDampingOverride=1
    angularSpringConstant=2000
    bAngularSpring=true
    bOverrideAngularDamping=true
    bUseGravityOverride=true
    simulationSpace=RootRelative
    alphaBoolBlend {blendOption=Linear}
    boundBone {boneName=arm_L}
    componentAppliedLinearAccClamp=<7000 7000 7000>
    componentLinearAccScale=<0.7 0.7 0.25>
    componentLinearVelScale=<0.8 0.8 0.5>
    gravityOverride {z=-500}
    physicsBodyDefinitions [{
      boundBone {boneName=arm_L}
      boxExtents=<10 10 15>
      constraintSetup {angularLimitsMax=<5 25 25>, angularLimitsMin=<-5 -25 -25>, angularTarget=<1 0 0>}
      localJointOffset=<0.3 0 0.2>
     }]
   }
   n5 = Anim Dynamics - Bone: arm_R!(ComponentPose=@n4, Alpha=1) -> (Pose) {
    angularBiasOverride=1
    angularDampingOverride=1
    angularSpringConstant=800
    bAngularSpring=true
    bOverrideAngularDamping=true
    bUseGravityOverride=true
    numSolverIterationsPreUpdate=3
    simulationSpace=RootRelative
    alphaBoolBlend {blendOption=Linear}
    boundBone {boneName=arm_R}
    componentAppliedLinearAccClamp=<5000 5000 5000>
    componentLinearAccScale=<0.7 0.7 0.25>
    componentLinearVelScale=<0.8 0.8 0.5>
    gravityOverride {z=-300}
    physicsBodyDefinitions [{
      boundBone {boneName=arm_R}
      boxExtents=<10 10 10>
      constraintSetup {angularLimitsMax=<5 25 25>, angularLimitsMin=<-5 -25 -25>, angularTarget=<1 0 0>}
      localJointOffset=<0.3 0.1 0>
     }]
   }
   n6 = Anim Dynamics - Bone: leg_L!(ComponentPose=@n5, Alpha=1) -> (Pose) {
    angularBiasOverride=1
    angularDampingOverride=1
    angularSpringConstant=800
    bAngularSpring=true
    bOverrideAngularDamping=true
    bUseGravityOverride=true
    numSolverIterationsPreUpdate=3
    simulationSpace=RootRelative
    alphaBoolBlend {blendOption=Linear}
    boundBone {boneName=leg_L}
    componentAppliedLinearAccClamp=<5000 5000 5000>
    componentLinearAccScale=<0.7 0.7 0.25>
    componentLinearVelScale=<0.8 0.8 0.5>
    gravityOverride {z=-300}
    physicsBodyDefinitions [{
      boundBone {boneName=leg_L}
      boxExtents=<10 10 10>
      constraintSetup {angularLimitsMax=<5 25 25>, angularLimitsMin=<-5 -25 -25>, angularTarget=<1 0 0>}
      localJointOffset=<0.3 0.1 0>
     }]
   }
   n7 = Anim Dynamics - Bone: leg_R!(ComponentPose=@n6, Alpha=1) -> (Pose) {
    angularBiasOverride=1
    angularDampingOverride=1
    angularSpringConstant=2000
    bAngularSpring=true
    bOverrideAngularDamping=true
    bUseGravityOverride=true
    simulationSpace=RootRelative
    alphaBoolBlend {blendOption=Linear}
    boundBone {boneName=leg_R}
    componentAppliedLinearAccClamp=<7000 7000 7000>
    componentLinearAccScale=<0.7 0.7 0.25>
    componentLinearVelScale=<0.8 0.8 0.5>
    gravityOverride {z=-500}
    physicsBodyDefinitions [{
      boundBone {boneName=leg_R}
      boxExtents=<10 10 15>
      constraintSetup {angularLimitsMax=<5 25 25>, angularLimitsMin=<-5 -25 -25>, angularTarget=<1 0 0>}
      localJointOffset=<0.5 0 0.2>
     }]
   }
   Output Pose(Result=Component To Local(ComponentPose=@n7)) {layerGroup=None}
  ]}
 EventGraph {nodes [
   TryGetPawnOwner()
   Event BlueprintUpdateAnimation()
  ]}
]
variables []
